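; Parent Blueprint for the Flaming Hoop, which launches players and vehicles passing through it, or hides and dies if hotfix-disabled.
; Overlap grants Nitro (playlist permitting) and camera effects, then activates knock abilities; Rocket Racing vehicles get their own speed boost instead of knockback.
; On-foot passes also trigger flare-up FX, quest messages and an optional cooldown, marked unused in BR.
; Replicated hoop state drives ring audio and material data.
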(event EventBeginPlay
  ;[1 1 1] OnBeginPlay - Check if Enabled
  (bind (_gamestate _playlist) (Class|FortAsyncActionCurrentPlaylistReady|PlaylistDataReadyAsync)
    (:OnReady
      (Variables|Default|SetPlaylist _playlist)
      (CallFunction|CheckPlaylistSafe)
      (if (== (ScalableFloat|GetValueatLevel (Variables|Default|GetHotfix_Enabled)) 1.0)
        ;[0 1 0.4077] Set up bindings and cast to vehicle
        (CallFunction|SetHoopState "NewEnumerator0")
        (else
          (Rendering|SetActorHiddenInGame true)
          (Gameplay|SilentDie true))))))

(event OnComponentBeginOverlap(BoostCollision)
  (params (OverlappedComponent :type Primitive%20Component%20Object%20Reference) (OtherActor :type Actor%20Object%20Reference) (OtherComp :type Primitive%20Component%20Object%20Reference) (OtherBodyIndex :type Integer) (bFromSweep :type Boolean) (SweepResult :type Hit%20Result%20Structure%20%28by%20ref%29))
  ;[0 1 0.4077] Set up bindings and cast to vehicle
  (CallFunction|HandleActorEntry OtherActor))

(event Custom|HandleActorEntry
  (params (EntryActor :type Actor%20Object%20Reference))
  ;[0 1 0.4077] Set up bindings and cast to vehicle
  (Utilities|CanUseHoop
    (:True
      (bind _output_get (Variables|Default|SetEntryActor EntryActor))
      (Variables|Default|SetEntryActorLocation (Transformation|GetActorLocation _output_get))
      (bind _asfort_athena_vehicle (Utilities|Casting|CastToFortAthenaVehicle _output_get)
        (:then
          ;[0 1 0.4077] Handle Vehicle Entry
          (bind _output_get_2 (Variables|Default|SetVehicle _asfort_athena_vehicle))
          (CallFunction|GrantGE _output_get_2)
          (Variables|Default|SetQuest_PlayersToUpdate (Vehicle|GetAllPassengers _asfort_athena_vehicle))
          (CallFunction|ApplyBoosttoVehicle)
          (GameplayCue|ExecuteGameplayCueOnActor(Burst) (Variables|Default|GetVehicle) (Variables|Default|GetGC_VehiclePassThrough) (Ability|GameplayCue|MakeGameplayCueParameters :NormalizedMagnitude 0.0 :RawMagnitude 0.0 :Location "0,0,0" :Normal "0,0,0" :Instigator (Variables|Default|GetVehicle) :GameplayEffectLevel 1 :AbilityLevel 1 :bReplicateLocationWhenUsingMinimalRepProxy false)))
        (:CastFailed
          ;[0 1 0.4077] Handle Pawn Entry
          (bind _asfort_pawn (Utilities|Casting|CastToFortPawn _output_get))
          (Utilities|IsinVehicle _asfort_pawn
            (:true
              (CallFunction|ForceGC))
            (:false
              (bind _isflying (CallFunction|CheckforCustomFlight _asfort_pawn))
              (if (not _isflying)
                (CallFunction|GrantGEToPlayer _asfort_pawn)
                (GameplayCue|ExecuteGameplayCueOnActor(Burst) _asfort_pawn (Variables|Default|GetGC_PassThrough) (Ability|GameplayCue|MakeGameplayCueParameters 0.0))
                (Variables|Default|SetQuest_PlayersToUpdate (Utilities|Array|MakeArray _asfort_pawn))
                (bind _bishoisted (CallFunction|CheckHoistedDBNO _asfort_pawn))
                (if (not _bishoisted)
                  (Tags|CancelAbilitiesWithTags (Ability|GetFortAbilitySystemComponent _asfort_pawn) (Variables|Default|GetCancelAbilitiesWithTags)) ; Some abilities need to be canceled BEFORE granting the knock ability, or their MME will cancel the impulse
                  (Ability|Advanced|GrantandActivateAbilityonSelfwithParams (Ability|GetFortAbilitySystemComponent _asfort_pawn) "/FlamingHoopsGameplay/Gameplay/Abilities/GA_Moonflax_FlamingHoops_Knock_Player.GA_Moonflax_FlamingHoops_Knock_Player_C" self (Utilities|Struct|MakeGameplayEventData "(TagName=\"Event.FlamingHoops.Launched.Player\")" self 0 0 0 "()" "(GameplayTags=)" "(GameplayTags=)" (Variables|Config|GetPlayerImpulseMagnitude)))
                  (CallFunction|BeginCooldown)
                  (CallFunction|OnHoopPassedThrough)
                  (CallFunction|FlareUp)
                  (else
                    _))
                (CallFunction|SendQuestUpdate (Variables|Default|GetQuest_PlayersToUpdate))))))))))

(event Custom|DoActive
  ;[1 0.3016 0.2038] Active again
  (switch Utilities|FlowControl|Switch|SwitchonEFlamingHoopState (Variables|Default|GetHoopState)
    (:NewEnumerator1
      _)
    (:NewEnumerator2
      _))
  (CallFunction|SetHoopState "NewEnumerator0"))

(event Custom|BeginCooldown
  ;[1 0.1228 0] Begin pending cooldown status - NOT IN USE IN BR
  (if (Variables|Default|GetbCooldownAllowed)
    (switch Utilities|FlowControl|Switch|SwitchonEFlamingHoopState (Variables|Default|GetHoopState)
      (:NewEnumerator0
        (CallFunction|SetHoopState "NewEnumerator1")
        (bind _returnvalue (Utilities|Time|SetTimerbyEvent (EventDispatchers|CreateEvent :!event Custom|DoCooldown) (ScalableFloat|GetValueatLevel (Variables|Default|GetHF_PreCooldown))))
        (Variables|Timers|SetTimer_DoCooldown _returnvalue)
        (Audio|SpawnSoundatLocation "/FlamingHoopsGameplay/Assets/Audio/FlamingHoop_Ring_Recharge_End_Cue.FlamingHoop_Ring_Recharge_End_Cue" (Transformation|GetWorldLocation (Variables|Transformation|GetRootComponent)))))))

(event Custom|OnActiveReplicated
  (bind _ringmesh (Variables|Default|GetRingMesh))
  ;[0 1 0.4077] Active visuals activated
  (Audio|Components|Audio|FadeIn (Variables|Default|GetSound_Ring) 1.0 1.0 0.0 "Linear" :!class /Script/Engine.AudioComponent)
  (Rendering|Material|SetCustomPrimitiveDataFloat _ringmesh 0 0.0) ; Tech art: set state and time stamp on material
  (Rendering|Material|SetCustomPrimitiveDataFloat _ringmesh 1 (Utilities|Time|GetGameTimeinSeconds)))

(event Custom|DoCooldown
  ;[1 0.1228 0] Apply cooldown status - NOT IN USE IN BR
  (switch Utilities|FlowControl|Switch|SwitchonEFlamingHoopState (Variables|Default|GetHoopState)
    (:NewEnumerator0
      _)
    (:NewEnumerator1
      _))
  (CallFunction|SetHoopState "NewEnumerator2")
  (bind _returnvalue (Utilities|Time|SetTimerbyEvent (EventDispatchers|CreateEvent :!event Custom|DoActive) (ScalableFloat|GetValueatLevel (Variables|Default|GetHF_CooldownTime))))
  (Variables|Timers|SetTimer_DoActive _returnvalue)
  (Audio|SpawnSoundatLocation "/FlamingHoopsGameplay/Assets/Audio/FlamingHoop_Ring_Recharge_Start_Cue.FlamingHoop_Ring_Recharge_Start_Cue" (Transformation|GetWorldLocation (Variables|Transformation|GetRootComponent))))

(event Custom|OnPendingReplicated
  ;[0 1 0.4077] Pending visuals activated
  (Audio|Components|Audio|FadeOut (Variables|Default|GetSound_Ring) 0.5 0.0 "Linear" :!class /Script/Engine.AudioComponent))

(event Custom|OnCooldownReplicated
  ;[0 1 0.4077] Cooldown visuals activated
  (Audio|Components|Audio|FadeOut (Variables|Default|GetSound_Ring) 0.0 0.0 "Linear" :!class /Script/Engine.AudioComponent)
  (Rendering|Material|SetCustomPrimitiveDataFloat (Variables|Default|GetRingMesh) 0 -1.0) ; Tech art: set state and time stamp on material
)

(event Custom|OnHoopPassedThrough
  (bind (_vector _forward) (CallFunction|CalculateEntryVector (Variables|Default|GetEntryActor)))
  (bind _ringmesh (Variables|Default|GetRingMesh))
  ;[0 1 0.4077] Hoop passed through
  (Networking|SwitchHasAuthority
    (:Authority
      _)
    (:Remote
      (Audio|SpawnSoundatLocation "/FlamingHoopsGameplay/Assets/Audio/FlamingHoop_Ring_Launch_Player_Cue.FlamingHoop_Ring_Launch_Player_Cue" (Transformation|GetWorldLocation (Variables|Transformation|GetRootComponent)))))
  (bind _output_get (Variables|Default|SetForward? _forward))
  (CallFunction|ActivateThroughFX _output_get)
  (Rendering|Material|SetCustomPrimitiveDataFloat _ringmesh 0 1.0) ; Tech art: set state and time stamp on material
  (Rendering|Material|SetCustomPrimitiveDataFloat _ringmesh 2 (Utilities|Time|GetGameTimeinSeconds)))

(event OnFxActorSpawned(PersistentFXComponent)
  (params (actor :type Actor%20Object%20Reference) (comp :type Persistent%20FXComponent%20Object%20Reference))
  ;[0.4163 0.6612 1] PersistentFX
  (Networking|SwitchHasAuthority
    (:Authority
      (bind _asbp_moonflax_flaming_hoop_persistent_fx (Utilities|Casting|CastToBP_Moonflax_FlamingHoop_PersistentFX actor))
      (Variables|PersistentFX|SetCurrentPersistentFXActor _asbp_moonflax_flaming_hoop_persistent_fx))))

(fn ActivateThroughFX
  (params (Forward? :type Boolean))
  (bind _flamefxdirectional (Variables|Default|GetFlameFXDirectional))
  ;[0.4895 0.1317 1] Activate fire when something passes through hoop
  (Effects|Components|ParticleSystem|SetBoolParameter (Variables|Default|GetSmokeFX) "Activate" false)
  (Effects|Components|ParticleSystem|SetFloatParameter _flamefxdirectional "Direction" (Math|Float|SelectFloat 1.0 -1.0 Forward?))
  (Transformation|SetWorldLocation (Variables|Default|GetFlamingHoopsThrough) (Variables|Default|GetEntryActorLocation) false false)
  (Components|Activation|Activate _flamefxdirectional true)
  (Effects|Components|ParticleSystem|SetBoolParameter (Variables|Default|GetSmokeFX) "Activate" true)
  (Utilities|Time|SetTimerbyEvent (EventDispatchers|CreateEvent :!event DeactivateThroughFX) 0.5))

(fn ApplyBoostToVehicle
  ;[0.824 0.6201 0.4987 0.5] The Rocket Racing vehicle actually loses speed when the hoop applies knockback, so boost it with its own system through the vehicle interface.
  ; This is applying the same bonus speed for the same duration as the Rocket Racing Speed Pad.
  (bind _vehicle (Variables|Default|GetVehicle)
    (:then
      (if (Class|FortVehicleInterface|IsDelMarVehicle _vehicle)
        (Class|FortVehicleInterface|TryBoostSpeedInternally _vehicle (GameplayTags|MakeLiteralGameplayTag "(TagName=\"FortDelMar.WorldBonusSpeed.SpeedPad\")") (LocalVariables|GetBonusSpeed_DelMarVehicle) (LocalVariables|GetBoostDuration_DelMarVehicle) true -1 _vehicle)
        (else
          (Ability|Advanced|GrantandActivateAbilityonSelfwithParams 0 "/FlamingHoopsGameplay/Gameplay/Abilities/GA_Moonflax_FlamingHoops_Knock_Vehicle.GA_Moonflax_FlamingHoops_Knock_Vehicle_C" self (Utilities|Struct|MakeGameplayEventData "(TagName=\"Event.FlamingHoops.Launched.Vehicle\")" self 0 0 0 "()" "(GameplayTags=)" "(GameplayTags=)" (Variables|Config|GetVehicleImpulseMultiplier))))))
    (:else)))

(fn CalculateEntryVector
  (params (EntryActor :type Actor%20Object%20Reference))
  ;[0.4895 0.1317 1] Find the direction of the entering actor
  (return (Math|Vector|SelectVector (Math|Vector|NegateVector (Transformation|GetForwardVector (Variables|Default|GetBoostCollision))) (Transformation|GetForwardVector (Variables|Default|GetBoostCollision)) (< (Math|Vector|DotProduct (Math|Vector|Normalize (Transformation|GetVelocity EntryActor)) (Transformation|GetForwardVector (Variables|Default|GetBoostCollision))) 0)) (not (< (Math|Vector|DotProduct (Math|Vector|Normalize (Transformation|GetVelocity EntryActor)) (Transformation|GetForwardVector (Variables|Default|GetBoostCollision))) 0))))

(fn CheckForCustomFlight
  (params (Pawn :type Fort%20Pawn%20Object%20Reference))
  (if (GameplayTags|HasMatchingGameplayTag Pawn "(TagName=\"MovementMode.Flying\")")
    (Ability|SendGameplayEventtoActor Pawn "(TagName=\"Event.Athena.Flying.DoInstantThrust\")")
    (return true)
    (else
      (return false))))

(fn CheckHoistedDBNO
  (params (FortPawn :type Fort%20Pawn%20Object%20Reference))
  (if (Health|IsDBNO FortPawn)
    (bind _asfort_player_pawn (Utilities|Casting|CastToFortPlayerPawn FortPawn)
      (:then
        (return (Utilities|IsValid (Class|FortPlayerPawn|GetDBNOHoister _asfort_player_pawn) :!class /Script/Engine.KismetSystemLibrary)))
      (:CastFailed
        _))
    (else
      _))
  (return false))

(fn CheckPlaylistSafe
  (bind _asfort_playlist_athena (Utilities|Casting|CastToFortPlaylistAthena (Variables|Default|GetPlaylist))
    (:then
      (bind _returnvalue (Class|FortPlaylistAthenaHelper|IsSpatialGameplayExperience _asfort_playlist_athena)) ; Is UEFN created island
      (if _returnvalue
        (Variables|Default|SetPlaylist_PermitsNitro true)
        (else
          (bind _returnvalue_2 (Class|FortPlaylistAthenaHelper|GetPlaylistType _asfort_playlist_athena))
          (switch Utilities|FlowControl|Switch|SwitchonEFortPlaylistType _returnvalue_2
            (:Default
              (Variables|Default|SetPlaylist_PermitsNitro false))
            (:Playground
              _)
            (:Creative
              _)
            (:Creative_LTM
              _))
          (Variables|Default|SetPlaylist_PermitsNitro true))))
    (:CastFailed
      (bind _returnvalue_3 (Class|FortPlaylistAthenaHelper|IsAnyCreativePlaylist _asfort_playlist_athena))
      (if _returnvalue_3
        (Variables|Default|SetPlaylist_PermitsNitro true)
        (elif (Fortnite|IsVkEditPlaylist)
          (Variables|Default|SetPlaylist_PermitsNitro true)
          (elif (Fortnite|IsPapayaPlaylist)
            (Variables|Default|SetPlaylist_PermitsNitro true)))))))

(fn DeactivateThroughFX
  (Effects|Components|ParticleSystem|SetBoolParameter (Variables|Default|GetSmokeFX) "Activate" false))

(fn FlareUp
  (bind _currentpersistentfxactor (Variables|PersistentFX|GetCurrentPersistentFXActor))
  (Networking|SwitchHasAuthority
    (:Authority
      (Utilities|IsValid _currentpersistentfxactor
        (:Is%20Valid
          (Class|BPMoonflaxFlamingHoopPersistentFX|ActivateFXServer _currentpersistentfxactor))))))

(fn ForceGC
  (bind _returnvalue (Fortnite|Athena|GameState|GetGameStateAthena))
  (bind _returnvalue_2 (GameState|GetRealMatchTime _returnvalue))
  (Networking|CanExecuteCosmeticEvents
    (:True
      (if (ScalableFloat|GetValueasBool (Variables|Default|GetHF_EnableForceGC))
        (Utilities|IsValid _returnvalue
          (:Is%20Valid
            (if (> _returnvalue_2 (+ (Variables|Default|GetLastForceGCMatchTime) 2.0))
              (Variables|Default|SetLastForceGCMatchTime _returnvalue_2)
              (Class|FortMemoryBPUtils|ForceGarbageCollection false))))))))

(fn GrantGE
  (params (Target :type Fort%20Athena%20Vehicle%20Object%20Reference))
  (bind _returnvalue (Vehicle|GetVehicleAbilitySystemComponent Target))
  ;[0.4895 0.1317 1] Grant Nitro + camera GA to each passenger
  (if (Variables|Default|GetPlaylist_PermitsNitro)
    (GameplayEffects|ApplyGameplayEffectToTarget :self _returnvalue :GameplayEffectClass (Variables|Default|GetGE_NitroGrant) :Target _returnvalue :Level 0.0)
    (for _array_element (Vehicle|GetAllPassengers Target)
      (GameplayEffects|ApplyGameplayEffectToTarget :self (Ability|GetFortAbilitySystemComponent _array_element) :GameplayEffectClass (Variables|Default|GetGE_NitroGrant) :Target (Ability|GetFortAbilitySystemComponent _array_element) :Level 0.0)
      (Class|NitroGameplayAnalytics|RegisterNitroAnalytics (Fortnite|GetFortPlayerControllerfromActor _array_element) (Variables|Default|GetNitroAnalyticsSource))
      ;;[0.824 0.6201 0.4987 0.5] The DelMar (Rocket Racing) vehicle has a static camera angle that we don't want to mess with.
      (if (not (Class|FortVehicleInterface|IsDelMarVehicle (Variables|GetTarget)))
        (GameplayEffects|ApplyGameplayEffectToTarget :self (Ability|GetFortAbilitySystemComponent _array_element) :GameplayEffectClass (Variables|Default|GetGE_CameraGrant_Vehicle) :Target (Ability|GetFortAbilitySystemComponent _array_element) :Level 0.0)))
    (GameplayEffects|ApplyGameplayEffectToTarget :self _returnvalue :GameplayEffectClass (Variables|Default|GetGE_VehicleEffect) :Target _returnvalue :Level 0.0)
    (return)
    (else
      _)))

(fn GrantGEToPlayer
  (params (Target :type Fort%20Pawn%20Object%20Reference))
  (bind _returnvalue (Ability|GetFortAbilitySystemComponent Target))
  ;[0.4895 0.1317 1] Nitro + Camera for a player who passes through the hoop
  (if (Variables|Default|GetPlaylist_PermitsNitro)
    (GameplayEffects|FortApplyGameplayEffectToTarget :self _returnvalue :GameplayEffectClass (Variables|Default|GetGE_NitroGrant) :TargetOfEffect _returnvalue :OptionalSourceObject self :Level 0.0)
    (Class|NitroGameplayAnalytics|RegisterNitroAnalytics (Fortnite|GetFortPlayerControllerfromActor Target) (Variables|Default|GetNitroAnalyticsSource))
    (else
      _))
  (GameplayEffects|ApplyGameplayEffectToTarget :self _returnvalue :GameplayEffectClass (Variables|Default|GetGE_CameraGrant) :Target _returnvalue :Level 0.0)
  (Variables|/Script/FortniteGame.FortPawn|SetbIgnoreNextFallingDamage true Target)
  (return))

(fn OnRep_EntryActor)

(fn OnRep_EntryActorLocation)

(fn OnRep_Forward?)

(fn OnRep_HoopState
  (bind _hoopstate (Variables|Default|GetHoopState))
  ;[0.4895 0.1317 1] Replicate state for players out of relevancy
  (Debugging|PrintEnum _hoopstate
    (:Then
      (switch Utilities|FlowControl|Switch|SwitchonEFlamingHoopState _hoopstate
        (:NewEnumerator0
          (CallFunction|OnActiveReplicated))
        (:NewEnumerator1
          (CallFunction|OnPendingReplicated))
        (:NewEnumerator2
          (CallFunction|OnCooldownReplicated)))
      (return))))

(fn SendQuestUpdate
  (params (PlayersToUpdate :type Array%20of%20Fort%20Pawn%20Object%20References))
  (for _array_element PlayersToUpdate
    (CustomVerb|BroadcastCustomVerbMessage (CustomVerb|MakeCustomVerbMessage _array_element "(TagName=\"CustomVerb.FlamingHoop\")")))
  (return))

(fn SetHoopState
  (params (HoopState :type EFlamingHoopState%20Enum :default "NewEnumerator0"))
  (Variables|Default|SetHoopState HoopState))

(fn ConstructionScript)
