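; Persistent FX static mesh actor for the Moonflax flaming hoop that plays a flare-up effect on demand.
; On BeginPlay the authority clears Always Relevant, which is only needed initially so the actor spawns regardless of player location.
; ActivateFXServer calls ActivateFXClients, which calls FlareUp to write current game time into the mesh's custom primitive data as a material timestamp.
; The editor-only spawn function and construction script are empty.
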(event EventBeginPlay
  (Networking|SwitchHasAuthority
    (:Authority
      (Variables|Replication|SetbAlwaysRelevant false) ; We need to be always relevant to start so we spawn no matter where the player is but we don't need it after that.
    )))

(event Custom|ActivateFXServer
  (CallFunction|ActivateFXClients))

(event Custom|ActivateFXClients
  (CallFunction|FlareUp))

(event Custom|FlareUp
  (Rendering|Material|SetCustomPrimitiveDataFloat (Variables|StaticMeshActor|GetStaticMeshComponent) 0 (Utilities|Time|GetGameTimeinSeconds)))

(fn editorOnly_SpawnFXActor
  (params (actorClass :type Actor%20Class%20Reference) (fxComp :type Persistent%20FXComponent%20Object%20Reference))
  (return 0))

(fn ConstructionScript)
